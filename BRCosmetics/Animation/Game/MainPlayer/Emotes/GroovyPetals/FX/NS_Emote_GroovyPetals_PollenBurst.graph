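# Niagara system for the Groovy Petals emote that releases a small puff of yellow-orange pollen sprites.
# A single Pollen emitter bursts a few randomly sized sprites from a small sphere in a wide upward cone.
# Light gravity, strong drag and curl noise make them drift, while curves fade their low alpha and pop their size in and out.
# System scope also references a foot-dust color-index module and front-end/surface user parameters.
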NiagaraSystem /BRCosmetics/Animation/Game/MainPlayer/Emotes/GroovyPetals/FX/NS_Emote_GroovyPetals_PollenBurst
dependencies {
 usedDataInterfaces [&NiagaraDataInterfaceColorCurve, &NiagaraDataInterfaceCurve]
 usedDynamicInputs []
 usedModules [
  &SystemState
  &self:SetFootDustFXColorArrayIndex
 ]
}
emitters {Pollen {
  emitterUpdateScript [
   &EmitterState(Life Cycle Mode=::NewEnumerator1, Loop Behavior=::NewEnumerator1, Loop Duration=1, UseLoopDelay=false)
   &SpawnBurst_Instantaneous(Spawn Count=5, Use Spawn Probability=false, UseLoopCountLimit=false)
  ]
  particleSpawnScript [
   &InitializeParticle(
    Lifetime Mode=::NewEnumerator1
    Mass Max=1.5
    Mass Mode=::NewEnumerator3
    Sprite Rotation Mode=::NewEnumerator2
    Sprite Size Mode=::NewEnumerator1
    Uniform Sprite Size Max=50
    Uniform Sprite Size Min=25
    UsePositionOffset=false
    Write Material Random=true
   )
   &ShapeLocation(MaskSpawnGroup=false, Override Seed=false, Sphere Radius=5, UseOwnerScale=true)
   &AddVelocity(
    Cone Angle=75
    Cone Axis=<0 0 1>
    Module.UseConeFalloff=false
    Module.UseVelDistribution=false
    Override Seed=false
    Rotation Coordinate Space=::World
    Velocity Mode=::NewEnumerator2
    Velocity Speed=&RandomRangeFloat(Minimum=25, Maximum=150)
   )
  ]
  particleUpdateScript [
   &ParticleState()
   &GravityForce(Gravity=<0 0 -25>)
   &Drag(Drag=&RandomRangeFloat(Minimum=6, Maximum=8), Use Linear Drag=true, Use Rotational Drag=true)
   &Color(Color=&ColorFromCurve(ColorCurve=DataInterface() {propertyValues {
       AlphaCurve {keys [<at 0 0.2 RCIM_Cubic>,<at 1 0 RCIM_Cubic>]}
       BlueCurve {keys [<at 0 0 RCIM_Cubic>]}
       GreenCurve {keys [<at 0 0.6932 RCIM_Cubic>]}
       RedCurve {keys [<at 0 1 RCIM_Cubic>]}
      }}))
   &CurlNoiseForce(Noise Strength=100, Pan Noise=false)
   &SolveForcesAndVelocity(Clamp Velocity=false, Limit Acceleration=false)
   &ScaleSpriteSize(Uniform Curve Sprite Scale=DataInterface() {propertyValues {Curve {keys [
        {arriveTangent=14.5835, interpMode=RCIM_Cubic, leaveTangent=14.5835, tangentMode=RCTM_User, time=0, value=0}
        {arriveTangent=16.2546, interpMode=RCIM_Cubic, leaveTangent=0.0003, tangentMode=RCTM_Break, time=0.1, value=1}
        {arriveTangent=-0.2714, interpMode=RCIM_Cubic, leaveTangent=-3.4346, tangentMode=RCTM_Break, time=0.9, value=1}
        {arriveTangent=-20.5076, interpMode=RCIM_Cubic, leaveTangent=-20.5076, tangentMode=RCTM_User, time=1, value=0}
       ]}}})
   &AlignSpriteToMeshOrientation(SetAlignment=true, SetFacing=true)
  ]
  rendererClasses [&NiagaraSpriteRendererProperties]
  renderers [{rendererClass=&NiagaraSpriteRendererProperties}]
 }}
summary {
 systemName=NS_Emote_GroovyPetals_PollenBurst
 emitters [{emitterName=Pollen, rendererClasses [&NiagaraSpriteRendererProperties]}]
 userVariables [
  User.bIsFrontEnd {type=NiagaraBool}
  User.bIsFrontEndPreview {type=NiagaraBool}
  User.FootDustColorArray {
   type=NiagaraDataInterfaceArrayColor
   description=""
   defaultValue {
    struct=NiagaraExt_VariableValue_DataInterface
    value {
     dataInterface=&self:NiagaraDataInterfaceArrayColor_0
     dataInterfaceClass=&NiagaraDataInterfaceArrayColor
    }
   }
  }
  User.Global Scale {type=NiagaraFloat, defaultValue=0, description=""}
  User.SurfaceTypeIndex {type=NiagaraInt32}
 ]
}
