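# Niagara system for the small dust puff in the GroovyPetals player emote, built from three burst-spawned sprite emitters.
# Smoke_Ground pushes a ring of flipbook smoke outward, slowed by drag and fading as it grows.
# SmallBits throws spinning debris upward under gravity, with collision disabled and size switched for front end, while DustSpikes adds a single short-lived, stretching streak.
# User parameters expose surface type, dust color array, front-end flags and global scale.
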NiagaraSystem /BRCosmetics/Animation/Game/MainPlayer/Emotes/GroovyPetals/FX/NS_Emote_GroovyPetals_SmallDustPuff
dependencies {
 usedDataInterfaces [
  &NiagaraDataInterfaceCamera
  &NiagaraDataInterfaceCurve
  &NiagaraDataInterfaceColorCurve
  &NiagaraDataInterfaceVector2DCurve
  &NiagaraDataInterfaceSpriteRendererInfo
 ]
 usedDynamicInputs [&Vector2DFromFloat]
 usedModules [
  &SystemState
  &self:SetFootDustFXColorArrayIndex
 ]
}
emitters {
 DustSpikes {
  emitterUpdateScript [
   &EmitterState()
   &SpawnBurst_Instantaneous(Spawn Count=1, Use Spawn Probability=false)
  ]
  particleSpawnScript [
   &InitializeParticle(
    Color Maximum=<rgba 1 1 1 0.2>
    Color Minimum=<rgba 0.8594 0.8594 0.8594 0.2>
    Color Mode=::NewEnumerator2
    Lifetime=&RandomRangeFloat(Minimum=0.2, Maximum=0.25)
    Sprite Size Max=<1 25>
    Sprite Size Min=<1 20>
    Sprite Size Mode=::NewEnumerator2
    UsePositionOffset=false
    Write Material Random=true
   )
   &ShapeLocation(
    Cylinder Distribution=::NewEnumerator2
    Cylinder Height=0
    Cylinder Radius=0
    MaskSpawnGroup=false
    Rotation Coordinate Space=::World
    Shape Primitive=::NewEnumerator1
    U Position=&ReturnNormalizedExecIndex(Normalized Index Includes One=false)
    UseOwnerScale=true
   )
   &AddVelocity(Velocity=&RandomRangeVector(Minimum=<0 0 1>, Maximum=<0 0 2>)) {module=AddVelocity001}
   &ApplyInitialForces()
   &CameraOffset(Camera Offset Amount=40)
  ]
  particleUpdateScript [
   &ParticleState()
   &InheritSourceMovement(ApplyForce=false, ApplyPosition=true, ApplyPositionFalloff=false)
   &ScaleSpriteSize(
    Non-Uniform Curve Sprite Scale=DataInterface() {propertyValues {
      XCurve {keys [{tangentMode=RCTM_User, time=0, value=1.1}, {tangentMode=RCTM_User, time=1, value=0}]}
      YCurve {keys [{tangentMode=RCTM_User, time=0, value=1}, {tangentMode=RCTM_User, time=1, value=2}]}
     }}
    Scale Sprite Size Mode=::NewEnumerator3
   )
   &ScaleColor(
    Scale Alpha=&FloatFromCurve(FloatCurve=DataInterface() {propertyValues {Curve {keys [
         {arriveTangent=-0.2156, interpMode=RCIM_Cubic, leaveTangent=-0.2156, tangentMode=RCTM_User, time=0, value=1}
         {arriveTangent=-0.9746, interpMode=RCIM_Cubic, leaveTangent=-0.9746, tangentMode=RCTM_User, time=1.0011, value=0}
        ]}}})
    Scale RGB=&MakeVectorFromLinearColorRGB(Color=&ColorFromCurve(ColorCurve=DataInterface() {propertyValues {
        AlphaCurve {keys [<at 0.3955 1 RCIM_Cubic>,<at 1 0 RCIM_Cubic>]}
        BlueCurve {keys [<at 0.358 0.7518>]}
        GreenCurve {keys [<at 0.358 1.1098>]}
        RedCurve {keys [<at 0.358 1.5>]}
       }}))
    ScaleA=true
    ScaleRGB=true
   )
   &SolveForcesAndVelocity(Clamp Velocity=false, Limit Acceleration=false)
   &SubUVAnimation(
    Evaluation Type=::NewEnumerator0
    "Sprite Renderer "=DataInterface() {propertyValues {SpriteRenderer=&self:DustSpikes_2.NiagaraSpriteRendererProperties_1}}
    SubUV Animation Mode:ENiagara_SubUVLookupModeV2=::NewEnumerator2
    UseEndFrame=false
    UseStartFrame=false
   )
   &DynamicMaterialParameters(
    Index 0 Param 1=0.5
    Index 0 Param 2=1
    Index 0 Param 3=&RandomRangeFloat(Minimum=0, Maximum=1)
    Index 0 Param 4=1
    Module.Param0WriteEnabled=true
    Module.Param1WriteEnabled=true
    Module.Param2WriteEnabled=true
    Module.Param3WriteEnabled=true
   )
  ]
  rendererClasses [&NiagaraSpriteRendererProperties]
  renderers [{rendererClass=&NiagaraSpriteRendererProperties}]
 }
 SmallBits {
  emitterUpdateScript [
   &EmitterState()
   &SpawnBurst_Instantaneous(
    Spawn Count=&RandomRangeInt(Minimum=8, Maximum=12)
    Use Spawn Probability=false
    UseLoopCountLimit=false
   )
  ]
  particleSpawnScript [
   /SetVariables(Particles.randomInt:NiagaraInt32=&RandomRangeInt(Minimum=0, Maximum=1)) {isSetParameters=true}
   &InitializeParticle(
    Color Maximum=<rgba 0.4323 0.4287 0.3936 5>
    Color Minimum=<rgba 0.1088 0.1094 0.0992 5>
    Color Mode=::NewEnumerator2
    Lifetime=&RandomRangeFloat(Minimum=0.5, Maximum=3)
    Sprite Rotation Mode=::NewEnumerator2
    Sprite Size Mode=::NewEnumerator3
    Uniform Sprite Size=&FN_FrontEndFloatModifier(
     Gameplay=&RandomRangeFloat(Minimum=0.5, Maximum=1.5)
     Front End=&RandomRangeFloat(Minimum=0.5, Maximum=2)
    )
    UsePositionOffset=false
    Write Material Random=true
   )
   &SubUVAnimation(End Frame=3, SubUV Animation Mode:ENiagara_SubUVLookupMode=::NewEnumerator2)
   &ShapeLocation(
    Cylinder Height=1
    Cylinder Radius=1
    MaskSpawnGroup=false
    Offset Mode=::NewEnumerator0
    Override Seed=false
    Shape Primitive=::NewEnumerator1
    Surface Only=false
    UseOwnerScale=true
   )
   /SetVariables(Particles.InitSpriteRotRate=&UniformRangedFloat(Minimum=-3, Maximum=3)) {isSetParameters=true}
   &AddVelocity(Velocity=&Multiply_VectorByFloat(Vector=$Particles.ShapeLocation.ShapeNormal, Float=&RandomRangeFloat(Minimum=75, Maximum=125)))
   &AddVelocity(Rotation Coordinate Space=::World, Velocity=&RandomRangeVector(Minimum=<0 0 250>, Maximum=<0 0 450>)) {module=AddVelocity001}
  ]
  particleUpdateScript [
   &ParticleState()
   &SpriteRotationRate(Rotation Rate=&Multiply_Float(
     A=$Particles.InitSpriteRotRate
     B=&FloatFromCurve(
      FloatCurve=DataInterface() {propertyValues {Curve {keys [
          {arriveTangent=-0.2085, interpMode=RCIM_Cubic, leaveTangent=-0.2085, tangentMode=RCTM_User, time=-0.0046, value=1.0113}
          {arriveTangent=-0.2496, interpMode=RCIM_Cubic, leaveTangent=-0.2496, tangentMode=RCTM_User, time=0.5, value=0.05}
         ]}}}
      CurveIndex=$Emitter.Age
      Scale Curve=950
     )
    ))
   &Collision(
    Age Colliding Particles=false
    Collision Enabled=&SetBoolByFloatComparison(A=$Emitter.Age, B=0.15)
    Control Roll On Collision=true
    Friction During a Bounce=0.2
    Module.Control Rest State Via Surface Normal=false
    Module.EnableMaxCollisionCount=false
    Module.Randomize Collision Normal Vector=true
    Restitution=0.5
    ScaleRateOfRotation=false
   ) {enabled=false}
   &GravityForce(Gravity=&RandomRangeVector(Minimum=<0 0 -980>, Maximum=<1 1 -1500>))
   &Drag(
    Drag=&FloatFromCurve(
     FloatCurve=DataInterface() {propertyValues {Curve {keys [{arriveTangent=-1, interpMode=RCIM_Cubic, leaveTangent=-3.4078, tangentMode=RCTM_Break, time=0, value=1}, <at 0.3 0.1>, <at 1 0>]}}}
     Scale Curve=7
    )
    Use Linear Drag=true
    Use Rotational Drag=false
   )
   &SolveForcesAndVelocity(Clamp Velocity=false, Limit Acceleration=false)
   &ScaleSpriteSize(
    Uniform Curve Scale=1.25
    Uniform Curve Sprite Scale=DataInterface() {propertyValues {Curve {keys [
        {arriveTangent=40, interpMode=RCIM_Cubic, time=0, value=0}
        {arriveTangent=0.1708, interpMode=RCIM_Cubic, leaveTangent=0.1708, tangentMode=RCTM_User, time=0.025, value=1}
        {arriveTangent=-3.2691, interpMode=RCIM_Cubic, leaveTangent=-3.2691, tangentMode=RCTM_User, time=1, value=0}
       ]}}}
   )
  ]
  rendererClasses [&NiagaraSpriteRendererProperties]
  renderers [{rendererClass=&NiagaraSpriteRendererProperties}]
 }
 Smoke_Ground {
  emitterUpdateScript [
   &EmitterState()
   &SpawnBurst_Instantaneous(Spawn Count=8, Use Spawn Probability=false)
  ]
  particleSpawnScript [
   &InitializeParticle(
    Color Maximum=<rgba 1 1 1 0.2>
    Color Minimum=<rgba 0.8594 0.8594 0.8594 0.2>
    Color Mode=::NewEnumerator2
    Lifetime=&RandomRangeFloat(Minimum=0.6, Maximum=0.7)
    Sprite Rotation Mode=::NewEnumerator2
    Sprite Size Mode=::NewEnumerator1
    UsePositionOffset=false
    Write Material Random=true
   )
   &ShapeLocation(
    Cylinder Distribution=::NewEnumerator2
    Cylinder Height=0
    Cylinder Radius=1
    MaskSpawnGroup=false
    Rotation Coordinate Space=::World
    Shape Primitive=::NewEnumerator1
    U Position=&ReturnNormalizedExecIndex(Normalized Index Includes One=false)
    UseOwnerScale=true
   )
   &AddVelocity(Velocity=&Multiply_VectorByFloat(Vector=$Particles.ShapeLocation.ShapeNormal, Float=&RandomRangeFloat(Minimum=50, Maximum=75)))
   &AddVelocity(Velocity=&RandomRangeVector(Minimum=<0 0 50>, Maximum=<0 0 75>)) {module=AddVelocity001}
   &ApplyInitialForces()
   &CameraOffset(Camera Offset Amount=40)
   /SetVariables(Particles.FirstFlipBookFrame:NiagaraInt32=&RandomRangeInt(Minimum=0, Maximum=64)) {isSetParameters=true}
   /SetVariables(
    Particles.Stored Particle Size:Vector3f=$Particles.Scale
    Particles.Stored Sprite Size:Vector2f=&Multiply_Vector2D_ByFloat(Float=1, Vector2D=$Particles.SpriteSize)
   ) {isSetParameters=true}
  ]
  particleUpdateScript [
   &ParticleState()
   /SetVariables(Particles.Scale=$StackContext.Stored Particle Size, Particles.SpriteSize=$StackContext.Stored Sprite Size) {isSetParameters=true}
   &AccelerationForce(Acceleration=&RandomRangeVector(Minimum=<0 0 -25>, Maximum=<1 1 -75>))
   &InheritSourceMovement(Applied Position Scale=<0.8 0.8 0.8>, ApplyForce=false, ApplyPosition=true, ApplyPositionFalloff=false)
   &self:ViewpointBased_Sprite_Rotation_01(Rotation Rate=3) {module=ViewpointBased_Sprite_Rotation_01}
   &Drag(
    Drag=&Multiply_Float(
     A=&RandomRangeFloat(Minimum=10, Maximum=12)
     B=&FloatFromCurve(FloatCurve=DataInterface() {propertyValues {Curve {keys [
          {arriveTangent=-2.2549, interpMode=RCIM_Cubic, leaveTangent=-2.2549, tangentMode=RCTM_User, time=0, value=1}
          {interpMode=RCIM_Cubic, tangentMode=RCTM_User, time=1, value=0}
         ]}}})
    )
    Use Linear Drag=true
    Use Rotational Drag=true
   )
   &ScaleSpriteSize(Uniform Curve Sprite Scale=DataInterface() {propertyValues {Curve {keys [{arriveTangent=4.1286, interpMode=RCIM_Cubic, leaveTangent=4.1286, tangentMode=RCTM_User, time=0, value=0.2}, <at 1 1.4 RCIM_Cubic>]}}})
   &DynamicMaterialParameters(
    Index 0 Param 1=&FloatFromCurve(FloatCurve=DataInterface() {propertyValues {Curve {keys [{arriveTangent=-2.5716, interpMode=RCIM_Cubic, leaveTangent=-2.5716, tangentMode=RCTM_User, time=0, value=1}, <at 0.6 0 RCIM_Cubic>]}}})
    Index 0 Param 4=1
    Module.Param0WriteEnabled=true
    Module.Param1WriteEnabled=true
    Module.Param2WriteEnabled=false
    Module.Param3WriteEnabled=true
   ) {module=DynamicMatParams}
   &ScaleColor(
    Scale Alpha=&FloatFromCurve(FloatCurve=DataInterface() {propertyValues {Curve {keys [
         {arriveTangent=-0.2156, interpMode=RCIM_Cubic, leaveTangent=-0.2156, tangentMode=RCTM_User, time=0, value=1}
         {arriveTangent=-0.9746, interpMode=RCIM_Cubic, leaveTangent=-0.9746, tangentMode=RCTM_User, time=1.0011, value=0}
        ]}}})
    Scale RGB=&MakeVectorFromLinearColorRGB(Color=&ColorFromCurve(ColorCurve=DataInterface() {propertyValues {
        AlphaCurve {keys [<at 0.3955 1 RCIM_Cubic>,<at 1 0 RCIM_Cubic>]}
        BlueCurve {keys [<at 0.0179 0.1262>,<at 0.9522 0.2448 RCIM_Cubic>]}
        GreenCurve {keys [<at 0.0179 0.2666>,<at 0.9522 0.517 RCIM_Cubic>]}
        RedCurve {keys [<at 0.0179 0.5156>,<at 0.9522 1 RCIM_Cubic>]}
       }}))
    ScaleA=true
    ScaleRGB=true
   )
   &SubUVAnimation(
    End Frame:NiagaraInt32=&Add_Integer(A=$Particles.FirstFlipBookFrame, B=64)
    Start Frame:NiagaraInt32=$Particles.FirstFlipBookFrame
    SubUV Loop Count:NiagaraFloat=&RandomRangeFloat(Minimum=0.25, Maximum=0.4)
   )
   &SolveForcesAndVelocity(Clamp Velocity=false, Limit Acceleration=false)
  ]
  rendererClasses [&NiagaraSpriteRendererProperties]
  renderers [{rendererClass=&NiagaraSpriteRendererProperties}]
 }
}
summary {
 systemName=NS_Emote_GroovyPetals_SmallDustPuff
 emitters [
  {emitterName=Smoke_Ground, rendererClasses [&NiagaraSpriteRendererProperties]}
  {emitterName=SmallBits, rendererClasses [&NiagaraSpriteRendererProperties]}
  {emitterName=DustSpikes, rendererClasses [&NiagaraSpriteRendererProperties]}
 ]
 userVariables [
  User.bIsFrontEnd {type=NiagaraBool}
  User.bIsFrontEndPreview {type=NiagaraBool}
  User.FootDustColorArray {
   type=NiagaraDataInterfaceArrayColor
   description=""
   defaultValue {
    struct=NiagaraExt_VariableValue_DataInterface
    value {
     dataInterface=&self:NiagaraDataInterfaceArrayColor_0
     dataInterfaceClass=&NiagaraDataInterfaceArrayColor
    }
   }
  }
  User.Global Scale {type=NiagaraFloat, defaultValue=0, description=""}
  User.SurfaceTypeIndex {type=NiagaraInt32}
 ]
}
